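; Sport motorcycle for AI convoy escorts, adjusting engine loudness for NPC riders and playing a VoidIn gameplay cue on the server at spawn.
; Clients set volume after a delay and playlist readiness; entering and exiting update it, using the NonParticipant tag to identify NPCs.
; UpdateEngineAudio fades the sportbike audio controller's engine component on cosmetic-capable machines.
; When an NPC exits, the server moves its AI radial leash to the bike's location.
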(event EventBeginPlay
  (bind _returnvalue (Vehicle|GetDriver))
  (|CallParentFunction)
  (Networking|SwitchHasAuthority
    (:Authority
      (GameplayCue|ExecuteGameplayCue self (Variables|Default|GetGC_VoidIn)))
    (:Remote
      (Utilities|FlowControl|Delay 0.5)
      (Class|FortAsyncActionCurrentPlaylistReady|PlaylistDataReadyAsync
        (:then
          _)
        (:OnReady
          (Utilities|IsValid _returnvalue
            (:Is%20Valid
              (CallFunction|UpdateEngineAudio (select (GameplayTags|HasMatchingGameplayTag _returnvalue "(TagName=\"Pawn.Athena.NonParticipant\")") (Variables|Default|GetNPCEngineVolume) 1.0))))))))
  (Rendering|SetVisibility (Variables|Vehicle|GetSkeletalMesh) true true))

(event EventOnPawnExitVehicle
  (params (PlayerPawn :type Fort%20Player%20Pawn%20Object%20Reference) (ExitSocketName :type Name :default "None"))
  (|CallParentFunction PlayerPawn ExitSocketName)
  (CallFunction|UpdateEngineAudio 1.0)
  (Networking|SwitchHasAuthority
    (:Authority
      (if (GameplayTags|HasMatchingGameplayTag PlayerPawn "(TagName=\"Pawn.Athena.NonParticipant\")")
        (bind _asfort_athena_aibot_controller (Utilities|Casting|CastToFortAthenaAIBotController (Pawn|GetController PlayerPawn)))
        (Class|FortAthenaAIBotController|SetRadialLeashLocation _asfort_athena_aibot_controller (Transformation|GetActorLocation self))))))

(event EventOnPawnEnterVehicle
  (params (PlayerPawn :type Fort%20Player%20Pawn%20Object%20Reference) (SeatIdx :type Integer))
  (|CallParentFunction PlayerPawn SeatIdx)
  (CallFunction|UpdateEngineAudio (select (GameplayTags|HasMatchingGameplayTag "(TagName=\"Pawn.Athena.NonParticipant\")") (Variables|Default|GetNPCEngineVolume) 1.0)))

(event Custom|UpdateEngineAudio
  (params (NewVolumeLevel :type Float%20%28single-precision%29))
  (bind _returnvalue (Vehicle|GetChildActorbyClass "/Sportbike/Audio/Settings/BP_Vehicles_AudioController_SportBike.BP_Vehicles_AudioController_SportBike_C"))
  (Networking|CanExecuteCosmeticEvents
    (:True
      (Utilities|IsValid _returnvalue
        (:Is%20Valid
          (Audio|Components|Audio|AdjustVolume (AudioController|GetEngineAudioComponent _returnvalue) 0.5 NewVolumeLevel "Linear" :!class /Script/Engine.AudioComponent))))))

(fn ConstructionScript
  (|CallParentFunction))
