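; Blueprint for the VividRazor AI convoy armored truck, carrying a vault and music staged by alert and vault progress.
; The first driver's alert changes drive music and its death spawns the vault; reinforcements run on a timer through a living world volume, stopping when another pawn enters or the vault opens.
; Server logic also toggles collision-damage mods, tags the truck disabled after rapid re-entry spam, and leashes exiting AI around inoperable trucks.
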(event EventBeginPlay
  (|CallParentFunction)
  (Utilities|FlowControl|Sequence
    (:then_0
      (CallFunction|BindEvents)
      (CallFunction|HandleIncreasedCollisionDamage false)
      (Default|SetMusicStage (Variables|Default|GetNPCMusicComponent_VividRazorVaults) (Variables|Default|GetMusic%20Stage)))
    (:then_1
      ; Void in VFX
      (Networking|SwitchHasAuthority
        (:Authority
          (GameplayCue|ExecuteGameplayCue self (Variables|Default|GetGC_VoidIn)))
        (:Remote
          (Utilities|FlowControl|Delay 0.5)
          (CallFunction|UpdateEngineAudio 0)))
      (Rendering|SetVisibility (Variables|Vehicle|GetSkeletalMesh) true true))))

(event Custom|BindEvents
  ; Bind Events to Enable/Disable Reinforcements
  (Variables|InternalData|SetReinforcementsStarted false)
  (Variables|InternalData|SetFirstDriverSet false)
  (Networking|SwitchHasAuthority
    (:Authority
      (bind _asfort_athena_living_world_volume (Utilities|Casting|CastToFortAthenaLivingWorldVolume (Variables|/Script/Engine.ChildActorComponent|GetChildActor (Variables|Default|GetLivingWorldVolume_ChildActor))))
      (Variables|Default|SetLivingWorldVolume _asfort_athena_living_world_volume)
      (Vehicle|BindEventtoOnPawnExitedVehiclefromSeat (EventDispatchers|CreateEvent :!event Custom|PawnLeftVehicle))
      (Vehicle|BindEventtoOnPawnEnteredVehicleatSeat (delegate Custom|PawnEnteredVehicle)))))

(event Custom|PawnEnteredVehicle
  (params (Vehicle :type Fort%20Vehicle%20Interface%20Interface%20%28by%20ref%29) (PlayerPawn :type Fort%20Pawn%20Object%20Reference) (SeatIndex :type Integer))
  ; Bind Events to Enable/Disable Reinforcements
  (if (Variables|InternalData|GetFirstDriverSet)
    (if (Variables|InternalData|GetReinforcementsStarted)
      (CallFunction|StopReinforcements)
      (else
        _))
    (if (Pawn|IsBotControlled PlayerPawn)
      (CallFunction|HandleIncreasedCollisionDamage false)
      (CallFunction|HandleVehicleDisablingonPawnEnterSpam PlayerPawn))
    (elif (== SeatIndex 0)
      (Variables|InternalData|SetFirstDriverSet true)
      (bind _output_get (Variables|InternalData|SetFirstDriver PlayerPawn))
      (bind _asfort_athena_aibot_controller (Utilities|Casting|CastToFortAthenaAIBotController (Pawn|GetController _output_get)))
      (Death|BindEventtoOnAgentDiedEvent _asfort_athena_aibot_controller (EventDispatchers|CreateEvent :!event Custom|FirstDriverHasDied))
      (AlertLevel|BindEventtoOnAlertLevelChangedEvent _asfort_athena_aibot_controller (delegate Custom|Alert%20Level%20Changed))))
  (CallFunction|UpdateEngineAudio 0))

(event Custom|PawnLeftVehicle
  (params (Vehicle :type Fort%20Vehicle%20Interface%20Interface%20%28by%20ref%29) (PlayerPawn :type Fort%20Pawn%20Object%20Reference) (SeatIndex :type Integer))
  ; Pawn Left Vehicle, Remove Increased Collision Damage
  (if (GameplayTags|HasMatchingGameplayTag PlayerPawn "(TagName=\"Pawn.Athena.NonParticipant\")")
    (CallFunction|HandleIncreasedCollisionDamage true)
    (CallFunction|UpdateEngineAudio PlayerPawn)
    ; Reset AI Patrol Data when they leave and the vehicle is inoperable
    (Networking|SwitchHasAuthority
      (:Authority
        (if (Vehicle|IsInoperable)
          (CallFunction|ResetAIPatrolData PlayerPawn))))))

(event Custom|Alert%20Level%20Changed
  (params (BotController :type Fort%20Athena%20AIBot%20Controller%20Object%20Reference) (OldAlertLevel :type EAlertLevel%20Enum :default "Unaware") (NewAlertLevel :type EAlertLevel%20Enum :default "Unaware"))
  (switch Utilities|FlowControl|Switch|SwitchonEAlertLevel NewAlertLevel
    (:Unaware
      _)
    (:LostTarget
      _)
    (:Threatened
      (Variables|Default|SetMusic%20Stage "NewEnumerator1")))
  (Variables|Default|SetMusic%20Stage "NewEnumerator0"))

(event EventStartReinforcements
  (params (Driver :type Fort%20Pawn%20Object%20Reference) (ReinforcementDuration :type Float%20%28double-precision%29) (LastTarget :type Actor%20Object%20Reference))
  (bind _livingworldvolume (Variables|Default|GetLivingWorldVolume))
  ; Start Spawning Reinforcements
  (|CallParentFunction Driver ReinforcementDuration 0)
  (Class|FortAthenaLivingWorldPointProviderInterface|EnablePointProvider_DEPRECATED _livingworldvolume)
  (Class|BPDynamicReinforcementsManager|StartReinforcements (Variables|Default|GetBP_DynamicReinforcementsManager) _livingworldvolume Driver LastTarget)
  (Variables|InternalData|SetReinforcementsStarted true)
  (Ability|SendGameplayEventtoActor Driver (Variables|Default|GetEventTag_ReinforcementsStarted))
  (bind _returnvalue (Utilities|Time|SetTimerbyEvent (EventDispatchers|CreateEvent :!event Custom|StopReinforcements) ReinforcementDuration))
  (Variables|InternalData|SetHNDL_TempVaultHackingDuration _returnvalue)
  (Class|AthenaSpecialActorComponent|SetMiniMapIcon (Variables|Default|GetAthenaSpecialActor) (Widget|Brush|MakeBrushfromTexture "/VividRazor_AIConvoy/Vehicles/Icons/T_UI_Minimap_Icon_ArmoredTruck_InProgress.T_UI_Minimap_Icon_ArmoredTruck_InProgress" 64 64)) ; Set to in progress
)

(event Custom|StopReinforcements
  ; Stop The Reinforcements
  (Utilities|Time|ClearandInvalidateTimerbyHandle (Variables|InternalData|GetHNDL_TempVaultHackingDuration))
  (Class|FortAthenaLivingWorldPointProviderInterface|DisablePointProvider (Variables|Default|GetLivingWorldVolume))
  (Class|BPDynamicReinforcementsManager|PauseReinforcements (Variables|Default|GetBP_DynamicReinforcementsManager)))

(event EventVaultOpened
  ; On Vault Open, Remove Damage Reducing GE, Stop Reinforcements and Remove MiniMap Icon
  (|CallParentFunction)
  (bind _returnvalue (GameplayEffects|RemoveActiveEffectswithTags (Ability|GetAbilitySystemComponent self) "(GameplayTags=((TagName=\"Granted.Athena.VividRazor.Convoy.AIReducedDamage\")))"))
  (if (== _returnvalue 0)
    (Fortnite|FortShippingLog self "Trying to remove GE that makes it so AI deal less damage to the Armored Convoy, but it failed." true)
    (else
      _))
  (CallFunction|StopReinforcements)
  (Components|Activation|SetActive (Variables|Default|GetAthenaSpecialActor) false false)
  (Variables|Default|SetMusic%20Stage "NewEnumerator3"))

(event Custom|FirstDriverHasDied
  (params (BotController :type Fort%20Athena%20AIBot%20Controller%20Object%20Reference) (AgentPawn :type Fort%20Pawn%20Object%20Reference))
  ; Driver has died for some reason, spawn the vault!
  (if (not (Variables|Default|GetbVaultHasSpawned))
    (CallFunction|SpawnVaultPrep)
    (Development|PrintString "Original Driver Has Died!")))

(event EventVaultGameplayStarted
  (|CallParentFunction)
  (Variables|Default|SetMusic%20Stage "NewEnumerator2"))

(event Custom|UpdateEngineAudio
  (params (PawnRef :type Fort%20Pawn%20Object%20Reference))
  (bind _returnvalue (Vehicle|GetChildActorbyClass "/ArmoredTruck/Assets/Audio/Controller/BP_Vehicle_AudioController_ArmoredTruck.BP_Vehicle_AudioController_ArmoredTruck_C"))
  (bind _returnvalue_2 (AudioController|GetEngineAudioComponent _returnvalue))
  (Networking|CanExecuteCosmeticEvents
    (:True
      (if (Vehicle|HasDriver)
        (Utilities|IsValid _returnvalue
          (:Is%20Valid
            (Audio|Parameter|SetFloatParameter _returnvalue_2 "REVVolume" (select (Pawn|IsBotControlled PawnRef) 1e+01 1.0) :!class /Script/Engine.AudioComponent)))
        (else
          (Utilities|IsValid _returnvalue
            (:Is%20Valid
              (Audio|Parameter|SetFloatParameter _returnvalue_2 "REVVolume" 1.0 :!class /Script/Engine.AudioComponent))))))))

(fn GetMusicStage
  (return (Variables|Default|GetMusic%20Stage)))

(fn HandleIncreasedCollisionDamage
  (params (Remove :type Boolean))
  (bind _tag_bossvehiclemods (Variables|Default|GetTag_BossVehicleMods))
  (Networking|SwitchHasAuthority
    (:Authority
      (if Remove
        (Vehicle|Mod|RemoveVehicleMod _tag_bossvehiclemods)
        (else
          (Vehicle|Mod|ApplyVehicleMod _tag_bossvehiclemods))))))

(fn HandleVehicleDisablingOnPawnEnterSpam
  (params (FortPawn :type Fort%20Pawn%20Object%20Reference))
  (bind _foundindex (LocalVariables|GetFoundIndex))
  (bind _pawnenteringmonitoring (Variables|InternalData|GetPawnEnteringMonitoring))
  (bind _output (Utilities|Array|Get(aref) _pawnenteringmonitoring _foundindex))
  (bind (_fortpawn_12_515c82024fd381295c0463a934b1611d _lastenteredtime_7_245f7c374434da5ce49615979c741547 _totalrecententeredtries_10_416db113435ea392690a1e8f37867791) (Utilities|Struct|BreakSTRUCTVividRaziorTruckPawnEnteringMonitor _output))
  (if (ScalableFloat|GetValueasBool (Variables|Default|GetEnableValidatingQuicklyEnteringVehicle))
    (LocalVariables|SetFoundPawn false)
    (for (_array_element _array_index) (Variables|InternalData|GetPawnEnteringMonitoring)
      (if (== FortPawn (Utilities|Struct|BreakSTRUCTVividRaziorTruckPawnEnteringMonitor _array_element))
        (LocalVariables|SetFoundPawn true)
        (LocalVariables|SetFoundIndex _array_index)))
    (if (LocalVariables|GetFoundPawn)
      (if (<= (- (Utilities|Time|GetGameTimeinSeconds) _lastenteredtime_7_245f7c374434da5ce49615979c741547) (ScalableFloat|GetValueatLevel (Variables|Default|GetTimeToleranceForQuicklyEnteringVehicle)))
        (bind _result (Math|Integer|IncrementInt _totalrecententeredtries_10_416db113435ea392690a1e8f37867791)
          (:%20%20
            (LocalVariables|SetNewAttemptCount _result)))
        (else
          (LocalVariables|SetNewAttemptCount 1)))
      (Variables|SetTempPawnEnteringMonitor_FortPawn_12_515C82024FD381295C0463A934B1611D (Utilities|Struct|MakeSTRUCTVividRaziorTruckPawnEnteringMonitor _fortpawn_12_515c82024fd381295c0463a934b1611d (Utilities|Time|GetGameTimeinSeconds) (LocalVariables|GetNewAttemptCount)))
      (Utilities|Array|SetArrayElem (Variables|InternalData|GetPawnEnteringMonitoring) (LocalVariables|GetFoundIndex) (LocalVariables|GetTempPawnEnteringMonitor))
      (if (>= (LocalVariables|GetNewAttemptCount) (ScalableFloat|GetValueatLevel (Variables|Default|GetNumberToleranceForQuicklyEnteringVehicle)))
        (Ability|GameplayEffect|AddLooseGameplayTags self (Variables|Default|GetGameplayTag_AIVehicleDisabled) true))
      (else
        (Utilities|Array|Add _pawnenteringmonitoring (Utilities|Struct|MakeSTRUCTVividRaziorTruckPawnEnteringMonitor (Variables|GetFortPawn) (Utilities|Time|GetGameTimeinSeconds) 1))))
    (else
      (return))))

(fn OnRep_MusicStage
  (bind _npcmusiccomponent_vividrazorvaults (Variables|Default|GetNPCMusicComponent_VividRazorVaults)
    (:then
      (Class|NPCMusicComponentVividRazorVaults|TransitionToNextMusicStage _npcmusiccomponent_vividrazorvaults (Variables|Default|GetMusic%20Stage)))
    (:else)))

(fn ResetAIPatrolData
  (params (Pawn :type Fort%20Pawn%20Object%20Reference))
  (bind _asfort_athena_aibot_controller (Utilities|Casting|CastToFortAthenaAIBotController (Pawn|GetController Pawn)))
  (Utilities|IsValid (Actor|GetComponentbyClass _asfort_athena_aibot_controller "/Script/FortniteAI.FortAthenaNpcPatrollingComponent")
    (:Is%20Valid
      (Class|FortAthenaNpcPatrollingComponent|RemovePatrolPath (Actor|GetComponentbyClass _asfort_athena_aibot_controller "/Script/FortniteAI.FortAthenaNpcPatrollingComponent")))
    (:Is%20Not%20Valid
      _))
  (Class|FortAthenaAIBotController|SetRadialLeashLocation _asfort_athena_aibot_controller (Transformation|GetActorLocation))
  (Class|FortAthenaAIBotController|SetRadialLeashInnerRadius _asfort_athena_aibot_controller 8e+02)
  (Class|FortAthenaAIBotController|SetRadialLeashOuterRadius _asfort_athena_aibot_controller 2.2e+03))

(fn ConstructionScript
  (|CallParentFunction))
